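# Niagara system for the Dark Viking male fire character cosmetic that, per its name, puts flame decals on the horns as orange HDR sprites.
# Three single-particle bursts use masked spawn groups to land at fixed offsets, two of them mirrored, with a camera offset.
# Particles outlive their lifetime.
# User.bisHandHidden zeroes their alpha, while User.bisHeadHidden sets their VisibilityTag; the other user parameters are unused here.
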NiagaraSystem /BRCosmetics/Effects/Fort_Effects/Effects/Characters/Athena_Parts/DarkVikingMaleFire/NS_DarkVikingMaleFire_Horns
dependencies {usedDataInterfaces [], usedDynamicInputs [], usedModules [&SystemState]}
emitters {FlameDecal {
  emitterUpdateScript [
   &EmitterState(Inactive Response=::NewEnumerator2, Life Cycle Mode=::NewEnumerator1, Loop Behavior=::NewEnumerator1, Loop Duration=1, UseLoopDelay=false)
   &SpawnBurst_Instantaneous(Spawn Count=1, Use Spawn Probability=false)
   &SpawnBurst_Instantaneous(Spawn Count=1, Spawn Group=1, Use Spawn Probability=false) {module=SpawnBurst_Instantaneous001}
   &SpawnBurst_Instantaneous(Spawn Count=1, Spawn Group=2, Use Spawn Probability=false) {module=SpawnBurst_Instantaneous002}
  ]
  particleSpawnScript [
   /Niagara/Modules/Spawn/Initialization/InitializeParticle(
    Color=<rgba 6 1.083 0 0.75>
    Mesh Scale=<10 10 10>
    Module.Write Color=true
    Module.Write Lifetime=true
    Module.Write Mass=false
    Module.Write Material Random=true
    Module.Write Position=false
    Module.Write Scale=true
    Module.Write SpriteRotation=false
    Module.Write SpriteSize=true
    Sprite Size=&Vector2DFromFloat(Value=10)
   )
   &CameraOffset(Camera Offset Amount=10)
   &BoxLocation(Box Offset=<25 -5 -9>, Box Size=<0 0 0>, MaskSpawnGroup=true, Module.Override Seed=false, Rotate Box=false, Surface Only=false)
   &BoxLocation(
    Box Offset=<25 -5 9>
    Box Size=<0 0 0>
    MaskSpawnGroup=true
    Module.Override Seed=false
    Rotate Box=false
    Spawn Group Mask=1
    Surface Only=false
   ) {module=BoxLocation001}
   &BoxLocation(
    Box Offset=<-8 12 0>
    Box Size=<0 0 0>
    MaskSpawnGroup=true
    Module.Override Seed=false
    Rotate Box=false
    Spawn Group Mask=2
    Surface Only=false
   ) {module=BoxLocation002}
  ]
  particleUpdateScript [
   &ParticleState(Kill Particles When Lifetime Has Elapsed=false)
   &ScaleColor(
    Scale Alpha=&MakeCustomFloatFromBool(Bool=$User.bisHandHidden, True Float=0, False Float=1)
    ScaleA=true
    ScaleRGB=true
   )
   /SetVariables(Particles.VisibilityTag=&MakeCustomIntFromBool(Bool=$User.bisHeadHidden, True INT=1, False INT=0)) {isSetParameters=true}
  ]
  rendererClasses [&NiagaraSpriteRendererProperties]
  renderers [{rendererClass=&NiagaraSpriteRendererProperties}]
 }}
summary {
 systemName=NS_DarkVikingMaleFire_Horns
 emitters [{emitterName=FlameDecal, rendererClasses [&NiagaraSpriteRendererProperties]}]
 userVariables [
  User.bIsFrontEndPreview {type=NiagaraBool}
  User.bIsFullyDeployed {type=NiagaraBool}
  User.bisHandHidden {type=NiagaraBool}
  User.bisHeadHidden {type=NiagaraBool}
  User.IsSprinting {type=NiagaraBool}
  User.PlayerSpeed {type=NiagaraFloat, defaultValue=0, description=""}
  User.RibbonIDOffset {type=NiagaraInt32}
 ]
}
